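# Physics asset for the F_MED_Jade backpack cosmetic, defining collision bodies and joints for its head, ear and tail bones.
# A sphere covers the head, while capsules cover the pack and the two segments of each ear and the tail.
# The head joins to the pack, and both ear chains and the tail chain branch from the head.
# Tight angular limits keep the ears and head fairly stiff, while the tail swings more widely.
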PhysicsAsset /BRCosmetics/Accessories/FORT_Backpacks/Backpack_F_MED_Jade/Meshes/F_MED_Jade_Pack_Physics
bodies {
 dyn_ear_l_01 [{shapeType=Capsule, length=3, radius=2, shapeName=dyn_ear_l_01_capsule, center=<0.5 0 0>, rotation=<rot 90 0 0>}]
 dyn_ear_l_02 [{shapeType=Capsule, length=2, radius=1.5, shapeName=dyn_ear_l_02_capsule, center=<0.5 0 0>, rotation=<rot 90 0 0>}]
 dyn_ear_r_01 [{shapeType=Capsule, length=3, radius=2, shapeName=dyn_ear_r_01_capsule, center=<0.5 0 0>, rotation=<rot 90 0 0>}]
 dyn_ear_r_02 [{shapeType=Capsule, length=2, radius=1.5, shapeName=dyn_ear_r_02_capsule, center=<0.5 0 0>, rotation=<rot 90 0 0>}]
 dyn_head [{shapeType=Sphere, radius=5, shapeName=dyn_head_sphere, center=<0 -1 0>}]
 dyn_tail_01 [{shapeType=Capsule, length=5, radius=5, shapeName=dyn_tail_01_capsule, center=<1 0 0>, rotation=<rot 90 0 0>}]
 dyn_tail_02 [{shapeType=Capsule, length=5, radius=5, shapeName=dyn_tail_02_capsule, center=<1 0 0>, rotation=<rot 90 0 0>}]
 pack [{shapeType=Capsule, length=3, radius=2, shapeName=pack_capsule}]
}
constraints [
 {bone1Name=dyn_tail_02, bone2Name=dyn_tail_01, swing1LimitDegrees=20, swing2LimitDegrees=10, twistLimitDegrees=5}
 {bone1Name=dyn_ear_r_02, bone2Name=dyn_ear_r_01, swing1LimitDegrees=8, swing2LimitDegrees=6, twistLimitDegrees=1}
 {bone1Name=dyn_ear_l_02, bone2Name=dyn_ear_l_01, swing1LimitDegrees=8, swing2LimitDegrees=6, twistLimitDegrees=1}
 {bone1Name=dyn_tail_01, bone2Name=dyn_head, swing1LimitDegrees=30, swing2LimitDegrees=20, twistLimitDegrees=5}
 {bone1Name=dyn_ear_r_01, bone2Name=dyn_head, swing1LimitDegrees=8, swing2LimitDegrees=6, twistLimitDegrees=1}
 {bone1Name=dyn_ear_l_01, bone2Name=dyn_head, swing1LimitDegrees=8, swing2LimitDegrees=6, twistLimitDegrees=1}
 {bone1Name=dyn_head, bone2Name=pack, swing1LimitDegrees=3, swing2LimitDegrees=1, twistLimitDegrees=10}
]
